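; Hookshot rope actor for the Thorn Ripper, built on FortHookGunRope, that reels back to the player and pulls along pickups the projectile overlapped.
; On reel end, the server tosses registered pickups near the rope toward the player on computed arcs, adding a component to stop sliding, then fires a gameplay event.
; A distance-scaled timeline drives rope material and projectile mesh cosmetics, then the server destroys the rope.
; Attachment to dead pawns is refused.
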(event EventBeginPlay
  ;[0.01493 0.07812 0.01331] BeginPlay: Bind to know when the reel has ended.
  (|CallParentFunction)
  (EventDispatchers|BindEventtoOnEndReel (EventDispatchers|CreateEvent :!event Custom|End%20Reel)))

(event Custom|End%20Reel
  ;[0.009944 0.1458 0.1328] EndReel: At this time, we should pull back the rope to the player pawn.
  (Networking|SwitchHasAuthority
    (:Authority
      (Pickups|PullRegisteredPickups)
      (Variables|Transient|SetbReelIn true)
      (Components|Timeline|SetPlayRate (Variables|Default|GetReel%20In) (CallFunction|GetReelinPlayRate))
      (bind (_direction _reelalpha _tightenalpha _flash) (Timeline|Reel%20In Play)
        (:Update
          (CallFunction|UpdateReelinCosmetics _reelalpha _tightenalpha _flash))
        (:Finished
          ;;[0.1927 0.1548 0.0465] On the server, destroy the rope at the end of reel in.
          ;; On the client, disable the actor collision and hide it.
          (Networking|SwitchHasAuthority
            (:Authority
              (Actor|DestroyActor))
            (:Remote
              (Rendering|SetActorHiddenInGame true)
              (Collision|SetActorEnableCollision false))))))))

(event EventEndPlay
  (params (EndPlayReason :type EEndPlayReason%20Enum :default "Destroyed"))
  ;[0.1302 0.08932 0.0255] EndPlay: Clean up.
  (Utilities|FlowControl|Sequence
    (:then_0
      (Networking|SwitchHasAuthority
        (:Authority
          (EventDispatchers|UnbindEventfromOnEndReel (EventDispatchers|CreateEvent :!event Custom|End%20Reel)))))
    (:then_1
      (|CallParentFunction EndPlayReason))))

(fn GetReelInPlayRate
  (bind _reelinplayratesettingsmap (Variables|Settings|GetReelInPlayRateSettingsMap))
  ;[0.09163 0.06863 0.1302] GetReelInPlayRate: Figure out the playrate of the timeline based on the reel in distance.
  (LocalVariables|SetReelInDistanceSq (Math|Vector|DistanceSquared(Vector) (Variables|FortHookGunRope|GetEndLocation) (Variables|FortHookGunRope|GetStartLocation)))
  (bind _keys (Utilities|Map|Keys _reelinplayratesettingsmap))
  (for _array_element _keys
    (if (.ReturnValue (Utilities|Map|Find _reelinplayratesettingsmap _array_element))
      (if (Math|Float|InRange(Float) (LocalVariables|GetReelInDistanceSq) (Math|Vector2D|BreakVector2D (Utilities|Map|Find _reelinplayratesettingsmap _array_element)) (.Y (Math|Vector2D|BreakVector2D (Utilities|Map|Find _reelinplayratesettingsmap _array_element))) true false)
        (return _array_element))))
  (return (LocalVariables|GetDefaultReturnPlayRate)))

(fn GetRopeLength
  ;[0.01229 0.1146 0.09508] GetRopeLength: Returns the distance between the start and end locations.
  (return (Math|Vector|Distance(Vector) (Variables|FortHookGunRope|GetStartLocation) (Variables|FortHookGunRope|GetEndLocation))))

(fn OnRep_bReelIn
  ;[0.83 0.2974 1] OnRep_bReelIn: Make the projectile mesh visible.
  ; Expected to be called only once when the variable is set to true.
  (Networking|CanExecuteCosmeticEvents
    (:True
      (Rendering|SetVisibility (Variables|StaticMesh|GetRopeProjectileMesh) (Variables|Transient|GetbReelIn) false))))

(fn PullRegisteredPickups
  (bind _returnvalue (Actor|GetComponentbyClass (LocalVariables|GetPickup) "/ThornRipper/Gameplay/Items/HookshotGameplay/Gameplay/B_HookshotPickupComponent.B_HookshotPickupComponent_C"))
  (bind _hotfix_pickuppulltossarc (Variables|Hotfix|GetHotfix_PickupPullTossArc))
  (bind _returnvalue_2 (ScalableFloat|GetValueatLevel _hotfix_pickuppulltossarc))
  (bind (_asfort_ammo_item_definition _bsuccess) (Utilities|Casting|CastToFortAmmoItemDefinition (Pickup|GetPrimaryItemDefinition (LocalVariables|GetPickup))))
  (bind _pickup (LocalVariables|GetPickup))
  ;[0.03917 0.09926 0.1458] PullRegisteredPickups: Toss the registered pickups back to the player.
  (Networking|SwitchHasAuthority
    (:Authority
      ;;[0.1406 0.03033 0.1401] Check to see if max number of pickups have been pulled.
      (bind _array_element (Utilities|Array|ForEachLoopwithBreak (Variables|Transient|GetRegisteredPickupsToPull))
        (:LoopBody
          (bind _output_get (LocalVariables|SetPickup _array_element))
          (Utilities|IsValid _output_get
            (:Is%20Valid
              (if (< (LocalVariables|GetNumOfPickupsPulled) (ScalableFloat|GetValueasInteger (Variables|Hotfix|GetHotfix_PickupPullMaxNumOfPickups)))
                _
                (else
                  _))
              ;;[0.5208 0.02925 0.4199] Check to see if pickup is still valid for pull and is still close enough to the rope.
              (if (Math|Boolean|ANDBoolean (not (Pickup|PickedUp (LocalVariables|GetPickup))) (not (Game|IsActorBeingDestroyed (LocalVariables|GetPickup))) (< (Math|Vector|GetPointDistancetoSegment (Transformation|GetActorLocation (LocalVariables|GetPickup)) (Variables|FortHookGunRope|GetStartLocation) (Variables|FortHookGunRope|GetEndLocation)) (ScalableFloat|GetValueatLevel (Variables|Hotfix|GetHotfix_PickupPullMaxValidDistFromRope))))
                (Utilities|FlowControl|Sequence
                  (:then_0
                    ;;[0.1686 0.2031 0.07688] Find the toss location for the pickup.
                    (LocalVariables|SetTossLocation (Variables|FortHookGunRope|GetStartLocation)) ; If player pawn is invalid, startlocation is the default backup.
                    (bind _asfort_player_pawn (Utilities|Casting|CastToFortPlayerPawn (Game|GetInstigator)))
                    (bind _output_get_2 (LocalVariables|SetTossLocation (+ (Transformation|GetActorLocation _asfort_player_pawn) (select _bsuccess "0, 0, 0" (+ (* (Math|Vector|GetUnitDirection(Vector) (Transformation|GetActorLocation _asfort_player_pawn) (Transformation|GetActorLocation (LocalVariables|GetPickup))) (ScalableFloat|GetValueatLevel (Variables|Hotfix|GetHotfix_PickupPullTossLocationDistFromPlayer))) (* (Math|Random|RandomUnitVector) (Math|Random|RandomFloatinRange 0.0 (ScalableFloat|GetValueatLevel (Variables|Hotfix|GetHotfix_PickupPullTossLocationVarianceDist)))))))))
                    (Rendering|Debug|DrawDebugSphere _output_get_2 1e+01 12 "(R=1.000000,G=1.000000,B=1.000000,A=1.000000)" 5.0 1.0))
                  (:then_1
                    ;;[1 0.3976 0.2168] Toss the projectile with the velocity that will have it reach the desired destination with the provided arc param.
                    (bind _outlaunchvelocity (Game|SuggestProjectileVelocityCustomArc (Transformation|GetActorLocation _pickup) (LocalVariables|GetTossLocation) 0.0 _returnvalue_2))
                    (Pickup|TossPickupinDirection _pickup _outlaunchvelocity 0 false false)
                    (Utilities|IsValid _returnvalue
                      (:Is%20Valid
                        (Class|BHookshotPickupComponent|DestroyComponent _returnvalue :!class /ThornRipper/Gameplay/Items/HookshotGameplay/Gameplay/B_HookshotPickupComponent.B_HookshotPickupComponent_C) ; Destroy any existing component.
                      )
                      (:Is%20Not%20Valid
                        _))
                    (Game|AddComponentbyClass _pickup "/ThornRipper/Gameplay/Items/HookshotGameplay/Gameplay/B_HookshotPickupComponent.B_HookshotPickupComponent_C" _outlaunchvelocity (LocalVariables|GetTossLocation) _returnvalue_2) ; This component is responsible for grinding the thrown \npickup to a halt once it hits the ground after the toss. \nOtherwise, the pickup slides across the ground going beyond the player.\nThis component cleans itself up after running its functionality.
                    (Math|Integer|IncrementInt (LocalVariables|GetNumOfPickupsPulled))))))))
        (:Completed
          _))
      (Pickups|ReportAnalyticsPickupsHooked (LocalVariables|GetNumOfPickupsPulled))
      (bind _asb_thorn_ripper (Utilities|Casting|CastToB_ThornRipper (Actor|GetOwner)))
      (Ability|SendGameplayEventtoActor (Variables|/Mantis/Items/B_Mantis_Weapon_Parent.B_Mantis_Weapon_Parent_C|GetPlayerCharacter _asb_thorn_ripper) (Variables|Default|GetPickupHookedEventTag) (Utilities|Struct|MakeGameplayEventData (Variables|Default|GetPickupHookedEventTag) 0 (Utilities|Array|Get(acopy) (Variables|Transient|GetRegisteredPickupsToPull)))))))

(fn RegisterPickupToPull
  (params (PickupActor :type Fort%20Pickup%20Object%20Reference))
  ;[0.04745 0.1198 0.1017] RegisterPickupToPull: Register the pickups that the projectile has overlapped on its path out.
  ; Called from B_Prj_Hookshot.
  (Networking|SwitchHasAuthority
    (:Authority
      (Utilities|IsValid PickupActor
        (:Is%20Valid
          (Utilities|Array|AddUnique (Variables|Transient|GetRegisteredPickupsToPull) PickupActor))))))

(fn ReportAnalyticsPickupsHooked
  (params (NumOfPickups :type Integer)))

(fn ShouldAttachToActor
  (params (InActor :type Actor%20Object%20Reference) (InReelAlpha :type Float%20%28single-precision%29))
  ; ShouldAttachToActor: Do not attach if the pawn has died.
  (bind _returnvalue (|CallParentFunction InActor InReelAlpha))
  (Utilities|FlowControl|Sequence
    (:then_0
      (if _returnvalue
        (bind _asfort_pawn (Utilities|Casting|CastToFortPawn InActor))
        (if (FortPawn|IsDead _asfort_pawn)
          (return false))))
    (:then_1
      (return _returnvalue))))

(fn UpdateRopeMeshMaterial
  (params (TightenAlpha :type Float%20%28double-precision%29) (Flash :type Float%20%28double-precision%29))
  ;[0.9324 1 0.2077] UpdateRopeMeshMaterial: Update the wiggle and flash params on the rope mesh material
  (Networking|CanExecuteCosmeticEvents
    (:True
      (bind _ropemesh (Variables|StaticMesh|GetRopeMesh)
        (:then
          (Rendering|Material|SetScalarParameterValueonMaterials _ropemesh "ReturnWiggle" (Math|Float|Lerp 0.0 3.0 TightenAlpha))
          (Rendering|Material|SetScalarParameterValueonMaterials _ropemesh "WiggleAmount" (Math|Float|Lerp 0.0 15.0 TightenAlpha))
          (Rendering|Material|SetScalarParameterValueonMaterials _ropemesh "Flash" Flash))
        (:else)))))

(fn UpdateRopeProjectileMeshLocationAndRotation
  (params (ReelAlpha :type Float%20%28double-precision%29))
  ;[0.09604 0.2188 0.1879] UpdateRopeProjectileMeshLocationAndRotation: Update the location of the mesh component that represents the projectile based on the reel alpha.
  (Networking|CanExecuteCosmeticEvents
    (:True
      (bind _ropeprojectilemesh (Variables|StaticMesh|GetRopeProjectileMesh)
        (:then
          (Transformation|SetWorldLocationAndRotation _ropeprojectilemesh (Math|Vector|Lerp(Vector) (Variables|FortHookGunRope|GetEndLocation) (Variables|FortHookGunRope|GetStartLocation) ReelAlpha) (Math|Rotator|FindLookatRotation (Variables|FortHookGunRope|GetEndLocation) (Variables|FortHookGunRope|GetStartLocation)) false false))
        (:else)))))

(fn ConstructionScript)
